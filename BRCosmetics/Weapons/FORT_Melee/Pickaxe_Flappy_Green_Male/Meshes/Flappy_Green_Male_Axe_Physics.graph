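# Physics asset for the Flappy_Green_Male pickaxe mesh, providing capsule bodies for the secondary motion of its cake and syrup bones.
# Top and bottom cake master bodies are angularly locked to the root and anchor the front and back cake pieces.
# Cake pieces have narrow, mostly single-axis swing and minimal twist.
# Two-link syrup chains hang off the top master, front and second back piece with wider swing limits for looser dangling.
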PhysicsAsset /BRCosmetics/Weapons/FORT_Melee/Pickaxe_Flappy_Green_Male/Meshes/Flappy_Green_Male_Axe_Physics
bodies {
 b_AR_Root [{shapeType=Capsule, length=0.505, radius=0.505, shapeName=b_AR_Root_capsule, rotation=<rot 0 -90 0>}]
 dyn_bottom_cakes_back [{shapeType=Capsule, length=3, radius=2.5, shapeName=dyn_bottom_cakes_back_capsule, center=<2.2423 0 0>, rotation=<rot -90 0 0>}]
 dyn_bottom_cakes_front [{shapeType=Capsule, length=4, radius=2.5, shapeName=dyn_bottom_cakes_front_capsule, center=<3.0174 0 0>, rotation=<rot -90 0 -0.0001>}]
 dyn_bottom_cakes_master [{shapeType=Capsule, length=20, radius=6, shapeName=dyn_bottom_cakes_master_capsule, center=<1.4019 0 0>, rotation=<rot 0 -90 0>}]
 dyn_top_cakes_back1 [{shapeType=Capsule, length=4, radius=2.5, shapeName=dyn_top_cakes_back1_capsule, center=<1.9163 0 0>, rotation=<rot -90 180 180>}]
 dyn_top_cakes_back2 [{shapeType=Capsule, length=12, radius=3, shapeName=dyn_top_cakes_back2_capsule, center=<7.0433 0 0>, rotation=<rot -90 0 0>}]
 dyn_top_cakes_back2_syrup1 [{shapeType=Capsule, length=4, radius=1, shapeName=dyn_top_cakes_back2_syrup1_capsule, center=<2.3812 0 0>, rotation=<rot -90 0 0>}]
 dyn_top_cakes_back2_syrup2 [{shapeType=Capsule, length=4, radius=1, shapeName=dyn_top_cakes_back2_syrup2_capsule, center=<2.8009 0 0>, rotation=<rot -90 0 0>}]
 dyn_top_cakes_front [{shapeType=Capsule, length=2.5, radius=2.5, shapeName=dyn_top_cakes_front_capsule, center=<1.9162 0 0>, rotation=<rot -90 0 0>}]
 dyn_top_cakes_front_syrup1 [{shapeType=Capsule, length=3, radius=1, shapeName=dyn_top_cakes_front_syrup1_capsule, center=<1.5953 0 0>, rotation=<rot -90 180 180>}]
 dyn_top_cakes_front_syrup2 [{shapeType=Capsule, length=3, radius=1, shapeName=dyn_top_cakes_front_syrup2_capsule, center=<2.4418 0 0>, rotation=<rot -90 180 180>}]
 dyn_top_cakes_master [{shapeType=Capsule, length=20, radius=5, shapeName=dyn_top_cakes_master_capsule, center=<0 0 -1.3432>, rotation=<rot 0 -86.7698 0>}]
 dyn_top_cakes_syrup1 [{shapeType=Capsule, length=3, radius=1, shapeName=dyn_top_cakes_syrup1_capsule, center=<1.8581 0 0>, rotation=<rot -90 180 180>}]
 dyn_top_cakes_syrup2 [{shapeType=Capsule, length=4, radius=1, shapeName=dyn_top_cakes_syrup2_capsule, center=<2.9082 0 0>, rotation=<rot -90 180 180>}]
}
constraints [
 {
  bone1Name=dyn_bottom_cakes_master
  bone2Name=b_AR_Root
  swing1LimitDegrees=45
  swing1Motion=Locked
  swing2LimitDegrees=45
  swing2Motion=Locked
  twistLimitDegrees=45
  twistMotion=Locked
 }
 {bone1Name=dyn_bottom_cakes_front, bone2Name=dyn_bottom_cakes_master, swing1LimitDegrees=12, swing2LimitDegrees=3, twistLimitDegrees=1}
 {bone1Name=dyn_bottom_cakes_back, bone2Name=dyn_bottom_cakes_master, swing1LimitDegrees=12, swing2LimitDegrees=3, twistLimitDegrees=1}
 {
  bone1Name=dyn_top_cakes_master
  bone2Name=b_AR_Root
  swing1LimitDegrees=45
  swing1Motion=Locked
  swing2LimitDegrees=45
  swing2Motion=Locked
  twistLimitDegrees=45
  twistMotion=Locked
 }
 {
  bone1Name=dyn_top_cakes_front
  bone2Name=dyn_top_cakes_master
  swing1LimitDegrees=16
  swing2LimitDegrees=8
  swing2Motion=Locked
  twistLimitDegrees=1
 }
 {bone1Name=dyn_top_cakes_front_syrup1, bone2Name=dyn_top_cakes_front, swing1LimitDegrees=35, swing2LimitDegrees=35, twistLimitDegrees=4}
 {
  bone1Name=dyn_top_cakes_front_syrup2
  bone2Name=dyn_top_cakes_front_syrup1
  swing1LimitDegrees=35
  swing2LimitDegrees=35
  twistLimitDegrees=4
 }
 {
  bone1Name=dyn_top_cakes_back1
  bone2Name=dyn_top_cakes_master
  swing1LimitDegrees=20
  swing2LimitDegrees=0
  swing2Motion=Locked
  twistLimitDegrees=0.5
 }
 {bone1Name=dyn_top_cakes_back2, bone2Name=dyn_top_cakes_master, swing1LimitDegrees=26, swing2LimitDegrees=1, twistLimitDegrees=1}
 {bone1Name=dyn_top_cakes_back2_syrup1, bone2Name=dyn_top_cakes_back2, swing1LimitDegrees=40, swing2LimitDegrees=40, twistLimitDegrees=2}
 {
  bone1Name=dyn_top_cakes_back2_syrup2
  bone2Name=dyn_top_cakes_back2_syrup1
  swing1LimitDegrees=30
  swing2LimitDegrees=30
  twistLimitDegrees=2
 }
 {bone1Name=dyn_top_cakes_syrup1, bone2Name=dyn_top_cakes_master, swing1LimitDegrees=35, swing2LimitDegrees=35, twistLimitDegrees=4}
 {bone1Name=dyn_top_cakes_syrup2, bone2Name=dyn_top_cakes_syrup1, swing1LimitDegrees=35, swing2LimitDegrees=35, twistLimitDegrees=4}
]
